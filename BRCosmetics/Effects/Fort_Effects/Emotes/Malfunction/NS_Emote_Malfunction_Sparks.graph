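# Niagara system for the Malfunction emote that throws short-lived orange spark sprites from a sampled location on the character's skeletal mesh.
# A single emitter mixes chance-based bursts with a continuous spawn rate, launching particles in randomized cones.
# Curl noise, ramping drag and gravity shape motion, while speed scales sprite size and time of day scales alpha.
# User float FXSparkStrength drives spawn rate and velocity; FXSparkColor, despite its name, scales color intensity.
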NiagaraSystem /BRCosmetics/Effects/Fort_Effects/Emotes/Malfunction/NS_Emote_Malfunction_Sparks
dependencies {
 usedDataInterfaces [
  &NiagaraDataInterfaceColorCurve
  &NiagaraDataInterfaceCurve
  &NiagaraDataInterfaceSkeletalMesh
  &NiagaraDataInterfaceArrayFloat
 ]
 usedDynamicInputs []
 usedModules [
  &self:SystemSpawnScript.SystemScriptSource.SystemScriptGraph.NiagaraNodeAssignment_0.SetVariables_#1
  &SystemState
  &Timeline
 ]
}
emitters {Sparks {
  emitterUpdateScript [
   &EmitterState()
   &SpawnBurst_Instantaneous(
    Spawn Count=&RandomRangeInt(Minimum=30, Maximum=44)
    Spawn Probability=0.25
    Use Spawn Probability=true
    UseLoopCountLimit=false
   )
   &SpawnRate(
    SpawnRate=&Multiply_Float(A=&RandomRangeFloat(Minimum=80, Maximum=130), B=$User.FXSparkStrength)
    Use Spawn Probability=false
   ) {module=SpawnRate001}
  ]
  particleSpawnScript [
   &InitializeParticle(
    Color Mode=::NewEnumerator0
    Lifetime Max=0.15
    Lifetime Min=0.06
    Lifetime Mode=::NewEnumerator1
    Sprite Size=&Multiply_Vector2D(
     A=&RandomRangeVector2D(Minimum=<1 2>, Maximum=<2 3.5>)
     B=&Vector2DFromFloat(Value=0.9)
    )
    Sprite Size Mode=::NewEnumerator4
    UsePositionOffset=false
    Write Material Random=true
   )
   &SkeletalMeshLocation(
    MaskSpawnGroup=false
    Mesh Sampling Type=::NewEnumerator5
    Sampled Position Offset=<0 -8 0>
    Skeletal Mesh=$System.Skeletal Mesh
    Socket Sampling Mode=::NewEnumerator5
   )
   &SphereLocation(
    MaskSpawnGroup=false
    Override Seed=false
    Sphere Radius=2
    Surface Only=false
    UseNonUniformScale=false
    UseOffset=false
    UseSphereOrientation=false
   )
   &AddVelocityInCone(
    Cone Angle=25
    Cone Axis=&RandomRangeVector(Minimum=<0.3 -1 0.5>, Maximum=<-0.3 -1 1>)
    Module.Use Velocity Falloff On Cone Axis=false
    Velocity Strength=&Multiply_Float(A=180, B=$User.FXSparkStrength)
   )
  ]
  particleUpdateScript [
   &ParticleState()
   &Color(
    Color=&ColorFromCurve(ColorCurve=DataInterface() {propertyValues {
       AlphaCurve {keys [<at 0 1 RCIM_Cubic>]}
       BlueCurve {keys [
         <at 0.1461 0.3646>
         {arriveTangent=-0.7238, interpMode=RCIM_Cubic, leaveTangent=-0.7238, time=0.3452, value=0.1719}
         <at 0.4913 0.1147 RCIM_Cubic>
        ]}
       GreenCurve {keys [
         <at 0.1461 6.2133>
         {arriveTangent=-16.6529, interpMode=RCIM_Cubic, leaveTangent=-16.6529, time=0.3452, value=1.3872}
         <at 0.4913 0.4649 RCIM_Cubic>
        ]}
       RedCurve {keys [<at 0.1461 10>, {arriveTangent=-26.0725, interpMode=RCIM_Cubic, leaveTangent=-26.0725, time=0.3452, value=3}, <at 0.4913 1 RCIM_Cubic>]}
      }})
    Scale Color=&VectorFromFloat(Value=&Multiply_Float(A=&RandomRangeFloat(Minimum=1, Maximum=3), B=$User.FXSparkColor))
   ) {module=Color001}
   /Niagara/Modules/Update/Forces/CurlNoiseForce(Noise Frequency=0.1, Noise Strength=1000, Pan Noise=true, Pan Noise Field=<0 0 1>)
   &Drag(
    Drag=&FloatFromCurve(
     FloatCurve=DataInterface() {propertyValues {Curve {keys [{interpMode=RCIM_Cubic, tangentMode=RCTM_User, time=0, value=0}, {interpMode=RCIM_Cubic, tangentMode=RCTM_User, time=1, value=1}]}}}
     Scale Curve=6
    )
    Use Linear Drag=true
    Use Rotational Drag=false
   )
   &ScaleColor(
    Color Value To Scale=$Particles.Color
    Scale Alpha=&Multiply_Float(
     A=&FloatFromCurve(FloatCurve=DataInterface() {propertyValues {Curve {keys [
          {arriveTangent=16.6122, interpMode=RCIM_Cubic, leaveTangent=16.6122, tangentMode=RCTM_User, time=0, value=0}
          <at 0.1 1 RCIM_Cubic>
          <at 1 0 RCIM_Cubic>
         ]}}})
     B=&Multiply_Float(A=1, B=&FNTimeOfDayScalar(Night Scale=0.7, Morning Scale=1, Evening Scale=1))
    )
    ScaleA=true
    ScaleRGB=true
   )
   &GravityForce(Gravity=<0 0 -400>) {module=GravityForce001}
   &SolveForcesAndVelocity(Clamp Velocity=false, Limit Acceleration=false)
   &ScaleSpriteSizeBySpeed(Max Scale Factor=<1 2>, Min Scale Factor=<1 0.5>)
  ]
  rendererClasses [&NiagaraSpriteRendererProperties]
  renderers [{rendererClass=&NiagaraSpriteRendererProperties}]
 }}
summary {
 systemName=NS_Emote_Malfunction_Sparks
 emitters [{emitterName=Sparks, rendererClasses [&NiagaraSpriteRendererProperties]}]
 userVariables [
  User.FXSparkColor {type=NiagaraFloat, defaultValue=1, description=""}
  User.FXSparkStrength {type=NiagaraFloat, defaultValue=1, description=""}
 ]
}
